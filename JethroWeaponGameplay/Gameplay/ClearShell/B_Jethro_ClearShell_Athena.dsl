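; ClearShell dual-hand weapon Blueprint that handles firing and overheat cosmetics for both hands.
; A timer started on heating, slower on mobile, feeds the heat ratio to medallion-aware Niagara effects on both meshes; overheating disables some emitters and rumbles, while cooling and equip changes tear effects down.
; Firing mirrors muzzle and shell effects to the other hand after a delay and drives a local heat audio layer.
; Bot-controlled pawns bypass the default fire effects as a workaround.
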(event EventBeginPlay
  (bind _returnvalue (Game|GetInstigator))
  ;[1 0.6311 0.9916 0.5] Event Begin Play
  ; Cache FortPlayerPawn
  (|CallParentFunction)
  (Utilities|IsValid _returnvalue
    (:Is%20Valid
      (bind _asfort_player_pawn (Utilities|Casting|CastToFortPlayerPawn _returnvalue))
      (Variables|Default|SetAs%20Fort%20Player%20Pawn _asfort_player_pawn))))

(event FortWeapon|EventOnEquippedAndAttached
  (bind _spawned_overheat_fx_righthand (Variables|Default|GetSpawned_Overheat_FX_RightHand))
  (bind _spawned_overheat_fx_lefthand (Variables|Default|GetSpawned_Overheat_FX_LeftHand))
  ;[0 0 0 0.5] Overheat Functionality
  ;;[0 1 0.4912] On Equip/Unequip -> Reset or else no Overheat FX spawn.
  (|CallParentFunction)
  (CallFunction|ResetFX)
  (Utilities|IsValid _spawned_overheat_fx_righthand
    (:Is%20Valid
      (Components|DestroyComponent _spawned_overheat_fx_righthand))
    (:Is%20Not%20Valid
      _))
  (Utilities|IsValid _spawned_overheat_fx_lefthand
    (:Is%20Valid
      (Components|DestroyComponent _spawned_overheat_fx_lefthand))))

(event FortWeapon|EventOnUnEquippedAndDetached
  (bind _spawned_overheat_fx_righthand (Variables|Default|GetSpawned_Overheat_FX_RightHand))
  (bind _spawned_overheat_fx_lefthand (Variables|Default|GetSpawned_Overheat_FX_LeftHand))
  ;[0 0 0 0.5] Overheat Functionality
  ;;[0 1 0.4912] On Equip/Unequip -> Reset or else no Overheat FX spawn.
  (|CallParentFunction)
  (CallFunction|ResetFX)
  (Utilities|IsValid _spawned_overheat_fx_righthand
    (:Is%20Valid
      (Components|DestroyComponent _spawned_overheat_fx_righthand))
    (:Is%20Not%20Valid
      _))
  (Utilities|IsValid _spawned_overheat_fx_lefthand
    (:Is%20Valid
      (Components|DestroyComponent _spawned_overheat_fx_lefthand))))

(event EventHandFired
  (params (Hand :type EDualWeaponHand%20Enum :default "LEFT") (bPersistantFire :type Boolean))
  (bind _returnvalue (Utilities|Enum|NotEqual(Enum) Hand "LEFT"))
  ;[1 1 0 0.5 32] Bot only-- prevent effects due to problems with the default setup.
  (if (not (Pawn|IsBotControlled (Variables|Actor|GetInstigator)))
    (|CallParentFunction Hand bPersistantFire)
    ;[0 0 0] Play FX on other Hand, should probably be rolled into base bp/class as an option, if it is already i did not see one
    (Utilities|FlowControl|Delay 0.1)
    (if (not (Variables|Default|GetbStoppedWeaponFireFX))
      (CallFunction|MuzzleFlashFX bPersistantFire _returnvalue)
      (if (and (not (Weapon|IsUsingScopeorFirstPersonCamera)) (> (Rendering|GetRenderingDetailMode) 0))
        (CallFunction|ActivateShellsFX (Variables|Effects%28Shells%29|GetUseShellsOnFire?) _returnvalue)
        (else
          _))
      (FortWeapon|PlayForceFeedback false (Variables|Default|GetbForceFeedbackShouldPersist)))))

(event Overheating|EventOnStartHeating
  ;[0 0 0 0.5] Overheat Functionality
  ;;[1 0.4469 0] Start Heating
  (|CallParentFunction)
  (bind _returnvalue (Utilities|Time|SetTimerbyEvent (EventDispatchers|CreateEvent :!event Custom|Start%20Overheat) (select (Fortnite|IsMobilePlatform) 0.1 0.03) true))
  (Variables|Default|SetOverheatFXTimer _returnvalue))

(event Custom|Start%20Overheat
  ;[0 0 0 0.5] Overheat Functionality
  ;;[1 0.4469 0] Start Heating
  (CallFunction|OverheatFX (Math|Conversions|ToByte(Integer) (Math|Float|Truncate (* 255 (/ (Overheat|GetCurrentOverheatValue) (Overheat|GetOverheatingMaxValue)))))))

(event EventOnStopWeaponFireFX
  ;[1 1 0 0.5 32] Bot only-- prevent effects due to problems with the default setup.
  (if (Pawn|IsBotControlled (Variables|Actor|GetInstigator))
    (CallFunction|DeactivateMuzzleFX)
    (CallFunction|DeactivateShellsFX)
    (else
      (|CallParentFunction))))

(event Custom|OverheatFX
  (params (Overheat%20Val :type Byte))
  (bind _returnvalue (Math|Conversions|ToInteger(Byte) Overheat%20Val))
  (bind _returnvalue_2 (Math|Conversions|ToFloat(Integer) _returnvalue))
  (bind _returnvalue_3 (/ _returnvalue_2 255.0))
  ;[0 0 0 0.5] Overheat Functionality
  ;;[1 0 0.1652] Overheat FX
  (if (> _returnvalue_3 0.25)
    (Utilities|DoOnceThenDo false
      (:DoOnce
        ;;[1 0 0.7965 0.5] Spawn Overheat FX
        ;;;[1 1 1] Right hand
        (bind _returnvalue_4 (Niagara|SpawnSystemAttached "/JethroWeaponGameplay/Assets/FX/Weapons/NS_Jethro_ClearShell_OverHeat.NS_Jethro_ClearShell_OverHeat" (Variables|Visual|GetWeaponMesh) "None" "0, 0, 0" "0, 0, 0" "KeepRelativeOffset" false false))
        (bind _output_get (Variables|Default|SetSpawned_Overheat_FX_RightHand _returnvalue_4))
        (Niagara|SetNiagaraVariable(Bool) _output_get "HasMedallion" (GameplayTags|HasTag (GameplayTags|GetOwnedGameplayTags (Variables|Default|GetAs%20Fort%20Player%20Pawn) :!class /Script/GameplayTags.BlueprintGameplayTagLibrary) (Variables|Default|GetMedallionTag)))
        ;;;[1 1 1] Left hand
        (bind _returnvalue_5 (Niagara|SpawnSystemAttached "/JethroWeaponGameplay/Assets/FX/Weapons/NS_Jethro_ClearShell_OverHeat.NS_Jethro_ClearShell_OverHeat" (CallFunction|GetLeftHandWeaponMesh) "None" "0, 0, 0" "0, 0, 0" "KeepRelativeOffset" false false))
        (bind _output_get_2 (Variables|Default|SetSpawned_Overheat_FX_LeftHand _returnvalue_5))
        (Niagara|SetNiagaraVariable(Bool) _output_get_2 "HasMedallion" (GameplayTags|HasTag (GameplayTags|GetOwnedGameplayTags (Variables|Default|GetAs%20Fort%20Player%20Pawn) :!class /Script/GameplayTags.BlueprintGameplayTagLibrary) (Variables|Default|GetMedallionTag))))
      (:ThenDo
        _))
    (CallFunction|ActivateOverheatFX (Variables|Default|GetSpawned_Overheat_FX_RightHand) _returnvalue_3)
    (CallFunction|ActivateOverheatFX (Variables|Default|GetSpawned_Overheat_FX_LeftHand) _returnvalue_3)))

(event Custom|ResetFX
  ;[0 0 0 0.5] Overheat Functionality
  ;;[1 0 0.1652] Overheat FX
  (Utilities|DoOnceThenDo false
    (:DoOnce
      ;;[1 0 0.7965 0.5] Spawn Overheat FX
      ;;;[1 1 1] Right hand
      (bind _returnvalue (Niagara|SpawnSystemAttached "/JethroWeaponGameplay/Assets/FX/Weapons/NS_Jethro_ClearShell_OverHeat.NS_Jethro_ClearShell_OverHeat" (Variables|Visual|GetWeaponMesh) "None" "0, 0, 0" "0, 0, 0" "KeepRelativeOffset" false false))
      (bind _output_get (Variables|Default|SetSpawned_Overheat_FX_RightHand _returnvalue))
      (Niagara|SetNiagaraVariable(Bool) _output_get "HasMedallion" (GameplayTags|HasTag (GameplayTags|GetOwnedGameplayTags (Variables|Default|GetAs%20Fort%20Player%20Pawn) :!class /Script/GameplayTags.BlueprintGameplayTagLibrary) (Variables|Default|GetMedallionTag)))
      ;;;[1 1 1] Left hand
      (bind _returnvalue_2 (Niagara|SpawnSystemAttached "/JethroWeaponGameplay/Assets/FX/Weapons/NS_Jethro_ClearShell_OverHeat.NS_Jethro_ClearShell_OverHeat" (CallFunction|GetLeftHandWeaponMesh) "None" "0, 0, 0" "0, 0, 0" "KeepRelativeOffset" false false))
      (bind _output_get_2 (Variables|Default|SetSpawned_Overheat_FX_LeftHand _returnvalue_2))
      (Niagara|SetNiagaraVariable(Bool) _output_get_2 "HasMedallion" (GameplayTags|HasTag (GameplayTags|GetOwnedGameplayTags (Variables|Default|GetAs%20Fort%20Player%20Pawn) :!class /Script/GameplayTags.BlueprintGameplayTagLibrary) (Variables|Default|GetMedallionTag))))
    (:ThenDo
      _))
  (CallFunction|ActivateOverheatFX (Variables|Default|GetSpawned_Overheat_FX_RightHand) (/ (Math|Conversions|ToFloat(Integer) (Math|Conversions|ToInteger(Byte) (stale Custom|OverheatFX Overheat%20Val))) 255.0))
  (CallFunction|ActivateOverheatFX (Variables|Default|GetSpawned_Overheat_FX_LeftHand) (/ (Math|Conversions|ToFloat(Integer) (Math|Conversions|ToInteger(Byte) (stale Custom|OverheatFX Overheat%20Val))) 255.0)))

(event Custom|SpawnFX)

(event Overheating|EventOnStartOverheated
  ;[0 0 0 0.5] Overheat Functionality
  ;;[1 0.1083 0] Start Overheated
  (|CallParentFunction)
  (CallFunction|SetNiagaraOverheatValues (Variables|Default|GetSpawned_Overheat_FX_RightHand))
  (CallFunction|SetNiagaraOverheatValues (Variables|Default|GetSpawned_Overheat_FX_LeftHand))
  (Game|Feedback|LocalPlayForceFeedback (Fortnite|GetFortPlayerControllerfromActor (Game|GetInstigator)) "/Game/Weapons/ForceFeedback/FF_Jethro_CalmRock_Overheated.FF_Jethro_CalmRock_Overheated" false "None"))

(event Overheating|EventOnStartCooling
  ;[0 0 0 0.5] Overheat Functionality
  ;;[0 1 0.8693] Start Cooling
  (|CallParentFunction)
  (Utilities|Time|ClearandInvalidateTimerbyHandle (Variables|Default|GetOverheatFXTimer))
  (Utilities|FlowControl|Delay 0.8)
  (Components|Activation|Deactivate (Variables|Default|GetSpawned_Overheat_FX_RightHand))
  (Components|Activation|Deactivate (Variables|Default|GetSpawned_Overheat_FX_LeftHand)))

(event Overheating|EventOnFinishedCooling
  ;[0 0 0 0.5] Overheat Functionality
  ;;[0 0.6897 1] Finished Cooling
  (|CallParentFunction)
  (Utilities|FlowControl|Delay 0.2)
  (Components|DestroyComponent (Variables|Default|GetSpawned_Overheat_FX_RightHand))
  (Components|DestroyComponent (Variables|Default|GetSpawned_Overheat_FX_LeftHand))
  (CallFunction|ResetFX))

(event EventOnPlayWeaponFireFX
  (params (bPersistentFire :type Boolean) (bSecondaryFire :type Boolean))
  (bind _overheataudiolayer (Variables|Default|GetOverheatAudioLayer))
  ;[0 0 0 0.5] Overheat Functionality
  ;;[1 0.9617 0 0.5] On Weapon Fire
  (if (Pawn|IsBotControlled (Variables|Actor|GetInstigator))
    ;;;[1 1 0 0.5 32] Bot only-- override effects due to problems with the default setup.
    (CallFunction|MuzzleFlashFX true)
    (CallFunction|MuzzleFlashFX true true)
    (CallFunction|ActivateShellsFX (Variables|Effects%28Shells%29|GetUseShellsOnFire?))
    (CallFunction|ActivateShellsFX (Variables|Effects%28Shells%29|GetUseShellsOnFire?) true)
    (else
      (|CallParentFunction bPersistentFire bSecondaryFire)
      (bind _as_fort_player_pawn (Variables|Default|GetAs%20Fort%20Player%20Pawn)
        (:then
          (if (Pawn|IsLocallyControlled _as_fort_player_pawn)
            (Audio|Components|Audio|Play _overheataudiolayer 0.0)
            (Audio|Components|Audio|SetVolumeMultiplier _overheataudiolayer (Math|Curves|GetFloatValue (Variables|Default|GetOverheatAudioFeedbackCurve) (/ (Overheat|GetCurrentOverheatValue) (Overheat|GetOverheatingMaxValue))) :!class /Script/Engine.AudioComponent)))
        (:else)))))

(fn ActivateOverheatFX
  (params (NewParam :type Niagara%20Particle%20System%20Component%20Object%20Reference) (Float :type Float%20%28double-precision%29))
  (Components|Activation|Activate NewParam false)
  (Niagara|SetNiagaraVariable(Float) NewParam "OverHeat" (Math|Float|Clamp(Float) (- (* Float 2.0) 1.0))))

(fn SetNiagaraOverheatValues
  (params (NiagaraSystemComponent :type Niagara%20Particle%20System%20Component%20Object%20Reference))
  (Effects|Components|ParticleSystem|SetEmitterEnable NiagaraSystemComponent "WeaponMeshArray" false)
  (Effects|Components|ParticleSystem|SetEmitterEnable NiagaraSystemComponent "Ambient_Steam" false)
  (Effects|Components|ParticleSystem|SetEmitterEnable NiagaraSystemComponent "MedallionFlare" false))

(fn ConstructionScript
  (|CallParentFunction))
